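# Niagara effect for the CabbageSugar emote that throws a burst of small meshes, worms by its name, upward in a narrow cone.
# One emitter with two mesh renderers applies toon-shader and outline parameters and time-of-day color.
# Meshes pop in and shrink out via a scale curve while gravity, drag and bouncy collision move them.
# Rotational velocity feeds mesh rotation force only after a particle has collided.
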NiagaraSystem /BRCosmetics/Animation/Game/MainPlayer/Emotes/CabbageSugar/FX/NS_CabbageSugar_WormsThrow
dependencies {
 usedDataInterfaces [&NiagaraDataInterfaceMeshRendererInfo, &NiagaraDataInterfaceCurve]
 usedDynamicInputs []
 usedModules [&SystemState]
}
emitters {UpwardMeshBurst {
  emitterSpawnScript [
   /Game/Effects/Niagara/MasterMaterials/ToonShader/Template/NM_OutlineParameters()
   /Game/Effects/Niagara/MasterMaterials/ToonShader/Template/NM_ToonShaderParameters(Lighting_Hardness=1)
  ]
  emitterUpdateScript [
   &EmitterState(Life Cycle Mode=::NewEnumerator1, Loop Behavior=::NewEnumerator1, Loop Duration=1, UseLoopDelay=false)
   &SpawnBurst_Instantaneous(Spawn Count=8, Use Spawn Probability=false, UseLoopCountLimit=false)
  ]
  particleSpawnScript [
   &InitializeParticle(
    Color=&FNTimeOfDayColor(
     Night Color=<rgba 0.3175 0.3839 0.4583 1>
     Morning Color=<rgba 0.6118 0.6392 0.5922 1>
     Day Color=<rgba 0.6118 0.6392 0.5922 1>
     Evening Color=<rgba 0.6118 0.6392 0.5922 1>
    )
    Lifetime=1.7
    Mass Max=1.1
    Mass Min=0.9
    Mass Mode=::NewEnumerator3
    Mesh Renderer Array Visibility Mode=::NewEnumerator1
    Mesh Renderer Info=DataInterface() {propertyValues {MeshRenderer=&self:UpwardMeshBurst.NiagaraMeshRendererProperties_7}}
    Mesh Scale Mode=::NewEnumerator3
    Mesh Uniform Scale=1.3
    UsePositionOffset=false
    Write Material Random=true
   )
   &ShapeLocation(MaskSpawnGroup=false, Override Seed=false, Sphere Radius=5, UseOwnerScale=true)
   &AddVelocity(
    Cone Angle=15
    Cone Axis=<0 0 1>
    Module.UseConeFalloff=true
    Module.UseVelDistribution=true
    Override Seed=false
    Speed Falloff From Cone Axis=0.875
    Velocity Mode=::NewEnumerator2
    Velocity Speed=&RandomRangeFloat(Minimum=450, Maximum=600)
   )
   &InitialMeshOrientation(Use Rotation Vector=false)
  ]
  particleUpdateScript [
   &ParticleState()
   &GravityForce()
   &UpdateMeshOrientation(Rotation Rate=2, Rotation Vector=&RandomRangeVector(Minimum=<-0.5 -0.5 -0.5>, Maximum=<0.5 0.5 0.5>))
   &Drag(Use Linear Drag=true, Use Rotational Drag=true)
   &ScaleMeshSize(Scale Factor=&VectorFromFloat(Value=&FloatFromCurve(FloatCurve=DataInterface() {propertyValues {Curve {keys [
          {arriveTangent=-0.0278, interpMode=RCIM_Cubic, time=0.0008, value=0}
          {arriveTangent=0.8536, time=0.1, value=1}
          {arriveTangent=-1.1111, interpMode=RCIM_Cubic, leaveTangent=-1.1111, time=0.9, value=1}
          <at 1 0 RCIM_Cubic>
         ]}}})))
   &Collision(
    Age Colliding Particles=false
    Control Roll On Collision=true
    KillOnCollision=false
    Module.Control Rest State Via Surface Normal=false
    Module.EnableMaxCollisionCount=false
    Module.Randomize Collision Normal Vector=false
    Radius Calculation Type=::NewEnumerator3
    Restitution=&RandomRangeFloat(Minimum=0.3, Maximum=0.4)
    ScaleRateOfRotation=true
   )
   &MeshRotationForce(Rotation=&Multiply_VectorByFloat(
     Vector=$Particles.RotationalVelocity
     Float=&MakeCustomFloatFromBool(Bool=$Particles.HasCollided, True Float=1, False Float=0)
    ))
   &SolveForcesAndVelocity(Clamp Velocity=false, Limit Acceleration=false)
  ]
  rendererClasses [&NiagaraMeshRendererProperties]
  renderers [
   {rendererClass=&NiagaraMeshRendererProperties, rendererIndex=0}
   {rendererClass=&NiagaraMeshRendererProperties, rendererIndex=1}
  ]
 }}
summary {
 systemName=NS_CabbageSugar_WormsThrow
 emitters [{emitterName=UpwardMeshBurst, rendererClasses [&NiagaraMeshRendererProperties]}]
 userVariables []
}
